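; Battle pass quest gameplay ability that tracks how far a physics boulder rolls after being damaged or dislodged from its destroyed spawner.
; A gameplay event selects the boulder, then a looping, randomly timed timer measures distance moved.
; Metres accumulate and reach the owning player's quest manager in batches; slowing ends the ability, flushing any sizeable remainder.
; The scripted Canary boulder skips tracking and reports a fixed distance.
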(event Ability|EventActivateAbilityFromEvent
  (params (EventData :type Gameplay%20Event%20Data%20Structure%20%28by%20ref%29))
  (bind (_eventtag _instigator _target _optionalobject _optionalobject2 _contexthandle _instigatortags _targettags) (Utilities|Struct|BreakGameplayEventData EventData))
  ;[0 0 0] Activate Ability - Check if target is a phys boulder, get its position, then start the timer loop
  (|CallParentFunction EventData)
  (bind _validtarget (CallFunction|IsValidTarget _eventtag _target _targettags))
  (if _validtarget
    (Development|PrintString "TRUE" true true "(R=1.000000,G=0.000000,B=0.364133,A=1.000000)")
    (Utilities|FlowControl|Delay 1.0)
    (bind _output_get (Variables|Default|SetTargetLocationPrevious (Transformation|GetActorLocation (Variables|Default|GetTarget))))
    (Variables|Default|SetTargetLocationCurrent _output_get)
    (bind _returnvalue (Utilities|Time|SetTimerbyEvent (EventDispatchers|CreateEvent :!event Custom|TargetDistanceMoved) (Math|Random|RandomFloatinRange 1.6 2.0) true))
    (Variables|Default|SetTimerHandle _returnvalue)
    (else
      (Ability|EndAbility))))

(event Custom|TargetDistanceMoved
  ;[0 0 0] Get the distance between the boulder's previous location in meters.. Send a batched update to quest manager when certain threshold is exceeded
  (bind _output_get (Variables|Default|SetTargetLocationCurrent (Transformation|GetActorLocation (Variables|Default|GetTarget))))
  (bind _output_get_2 (Variables|Default|SetMetersTravelled (/ (Math|Vector|Distance(Vector) _output_get (Variables|Default|GetTargetLocationPrevious)) 1e+02)))
  (if (> _output_get_2 3.0)
    (bind _output_get_3 (Variables|Default|SetMetersTravelledCached (+ (Variables|Default|GetMetersTravelledCached) (Variables|Default|GetMetersTravelled))))
    (if (> _output_get_3 2e+01)
      (CallFunction|SendQuestEvent (Variables|Default|GetMetersTravelledCached))
      (Variables|Default|SetMetersTravelledCached 0.0)
      (else
        _))
    (Variables|Default|SetTargetLocationPrevious (Variables|Default|GetTargetLocationCurrent))
    (Variables|Default|SetMetersTravelled 0.0)
    (else
      (Development|PrintString "Boulder slowing down, ending..." true true "(R=1.000000,G=0.000000,B=0.364133,A=1.000000)")
      (Ability|EndAbility))))

(event Ability|EventOnEndAbility
  (params (bWasCancelled :type Boolean))
  (bind _meterstravelledcached (Variables|Default|GetMetersTravelledCached))
  (bind _timerhandle (Variables|Default|GetTimerHandle))
  ;[0 0 0] End ability - send a final distance update if cached distance is less than the threshold when ended
  (|CallParentFunction bWasCancelled)
  (if (Utilities|Time|IsValidTimerHandle _timerhandle)
    (Utilities|Time|ClearandInvalidateTimerbyHandle _timerhandle)
    (if (> _meterstravelledcached 3.0)
      (CallFunction|SendQuestEvent _meterstravelledcached))))

(fn IsValidTarget
  (params (EventTag :type Gameplay%20Tag%20Structure) (TargetActor :type Actor%20Object%20Reference) (TargetTags :type Gameplay%20Tag%20Container%20Structure))
  (bind _returnvalue (GameplayTags|HasTag TargetTags "(TagName=\"Building.Type.Prop.Organic.PhysicsBoulder\")" true))
  ;[0 0 0] Check if the target that was destroyed or damaged is a boulder or boulder spawner
  (switch Utilities|FlowControl|Switch|SwitchonGameplayTag EventTag
    (:Default
      _)
    (:Event.Canary.Run
      ;;[0.09306 0.09306 0.09306] Special case for the Canary boulder, which is not technically a physics boulder and has a set path
      (CallFunction|SendQuestEvent 55.0)
      (return false))
    (:Event.Damage.Killed
      ;;[0.09375 0.09375 0.09375] If destroyed, see if target can be casted to boulder spawner, then set the boulder that was spawned as the valid target
      (if _returnvalue
        (bind _asprop_physics_fellable_parent (Utilities|Casting|CastToProp_PhysicsFellable_Parent TargetActor)
          (:then
            (Variables|Default|SetTarget (Variables|/Game/Athena/Items/PhysicsActors/PhysicsFellable/Prop_PhysicsFellable_Parent.Prop_PhysicsFellable_Parent_C|GetSpawnedPhysicsObject _asprop_physics_fellable_parent))
            (return true))
          (:CastFailed
            _)) ; Get the boulder object that was spawned
        (else
          _))
      (return false))
    (:Event.Damage.Dealt
      (if (and _returnvalue (not (GameplayTags|HasTag TargetTags "(TagName=\"Building.Type.Prop.FellablePhysicsProp\")" true)))
        (Variables|Default|SetTarget TargetActor)
        (return true)
        (else
          _)))))

(fn SendQuestEvent
  (params (Count :type Float%20%28double-precision%29))
  ;[0 0 0] Sends a complex custom to the owning player's quest manager using the cached meters travelled value
  (Quest|SendComplexCustomStatEvent (Variables|GAVars|Data|GetOwningPlayerQuestManager) 0 (Variables|Default|GetQuestSourceTag) "(GameplayTags=)" (Math|Float|Round Count)))
